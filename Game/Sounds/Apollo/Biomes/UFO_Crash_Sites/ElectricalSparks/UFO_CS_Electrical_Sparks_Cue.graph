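# Sound cue for electrical spark one-shots at UFO crash sites, layering close and distant variations.
# A mixer blends a louder close branch with a quieter distant branch, each using its own attenuation settings for distance-based layering.
# Each branch picks an equally weighted random wave and applies slight pitch and volume modulation.
# A concurrency override caps simultaneous instances.
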SoundCue /Game/Sounds/Apollo/Biomes/UFO_Crash_Sites/ElectricalSparks/UFO_CS_Electrical_Sparks_Cue
root=Mixer_0
nodes [
 {
  id=Attenuation_0
  children [Random_2]
  properties {attenuationSettings=&/Fort_Electrical_Spark_Close}
 }
 {
  id=Attenuation_1
  children [Random_1]
  properties {attenuationSettings=&/Fort_Electrical_Spark_Dist}
 }
 {id=Mixer_0, children [Modulator_0,Modulator_2], properties {inputVolume [0.75,0.2]}}
 {id=Modulator_0, children [Attenuation_0], properties {pitchMax=1.2, pitchMin=0.9, volumeMax=1.1}}
 {id=Modulator_2, children [Attenuation_1], properties {pitchMax=1.2, pitchMin=0.9, volumeMax=1.1, volumeMin=0.9}}
 {
  id=Random_1
  children [WavePlayer_12, WavePlayer_13, WavePlayer_14, WavePlayer_15, WavePlayer_16, WavePlayer_17]
  properties {weights [1,1,1,1,1,1]}
 }
 {
  id=Random_2
  children [
   WavePlayer_18
   WavePlayer_19
   WavePlayer_20
   WavePlayer_21
   WavePlayer_22
   WavePlayer_23
   WavePlayer_24
   WavePlayer_25
   WavePlayer_26
   WavePlayer_27
   WavePlayer_28
   WavePlayer_29
  ]
  properties {weights [1,1,1,1,1,1,1,1,1,1,1,1]}
 }
 {
  id=WavePlayer_12
  properties {soundWaveAssetPtr=&/UFO_CS_Electrical_Spark_Distant_01}
 }
 {
  id=WavePlayer_13
  properties {soundWaveAssetPtr=&/UFO_CS_Electrical_Spark_Distant_02}
 }
 {
  id=WavePlayer_14
  properties {soundWaveAssetPtr=&/UFO_CS_Electrical_Spark_Distant_03}
 }
 {
  id=WavePlayer_15
  properties {soundWaveAssetPtr=&/UFO_CS_Electrical_Spark_Distant_04}
 }
 {
  id=WavePlayer_16
  properties {soundWaveAssetPtr=&/UFO_CS_Electrical_Spark_Distant_05}
 }
 {
  id=WavePlayer_17
  properties {soundWaveAssetPtr=&/UFO_CS_Electrical_Spark_Distant_06}
 }
 {
  id=WavePlayer_18
  properties {soundWaveAssetPtr=&/UFO_CS_Electrical_Spark_Close_12}
 }
 {
  id=WavePlayer_19
  properties {soundWaveAssetPtr=&/UFO_CS_Electrical_Spark_Close_04}
 }
 {
  id=WavePlayer_20
  properties {soundWaveAssetPtr=&/UFO_CS_Electrical_Spark_Close_08}
 }
 {
  id=WavePlayer_21
  properties {soundWaveAssetPtr=&/UFO_CS_Electrical_Spark_Close_09}
 }
 {
  id=WavePlayer_22
  properties {soundWaveAssetPtr=&/UFO_CS_Electrical_Spark_Close_10}
 }
 {
  id=WavePlayer_23
  properties {soundWaveAssetPtr=&/UFO_CS_Electrical_Spark_Close_11}
 }
 {
  id=WavePlayer_24
  properties {soundWaveAssetPtr=&/UFO_CS_Electrical_Spark_Close_01}
 }
 {
  id=WavePlayer_25
  properties {soundWaveAssetPtr=&/UFO_CS_Electrical_Spark_Close_02}
 }
 {
  id=WavePlayer_26
  properties {soundWaveAssetPtr=&/UFO_CS_Electrical_Spark_Close_03}
 }
 {
  id=WavePlayer_27
  properties {soundWaveAssetPtr=&/UFO_CS_Electrical_Spark_Close_05}
 }
 {
  id=WavePlayer_28
  properties {soundWaveAssetPtr=&/UFO_CS_Electrical_Spark_Close_06}
 }
 {
  id=WavePlayer_29
  properties {soundWaveAssetPtr=&/UFO_CS_Electrical_Spark_Close_07}
 }
]
properties {
 bOverrideConcurrency=true
 maxDistance=5000
 soundClassObject=&Fort_SFX_InGame
 volumeMultiplier=1
 concurrencyOverrides {maxCount=3, voiceStealReleaseTime=0.25}
}
